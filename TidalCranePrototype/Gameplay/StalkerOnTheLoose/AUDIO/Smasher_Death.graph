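# Sound cue for the Smasher's death in the StalkerOnTheLoose gameplay, built as three mixed layers.
# A modulated random pick of four death sounds plays with a fixed sweetener wave and a briefly delayed, pitched-up dematerialize sound.
# The mix feeds only the first branch of a quality level root.
# Inline overrides set natural-sound distance attenuation and prevent new instances once the concurrency limit is reached.
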SoundCue /TidalCranePrototype/Gameplay/StalkerOnTheLoose/AUDIO/Smasher_Death
root=QualityLevel_1
nodes [
 {id=Delay_2, children [WavePlayer_18], properties {delayMax=0.5, delayMin=0.5}}
 {id=Mixer_2, children [Modulator_4, WavePlayer_19, Modulator_6], properties {inputVolume [1,1,1]}}
 {id=Modulator_4, children [Random_3]}
 {id=Modulator_6, children [Delay_2], properties {pitchMax=1.3, pitchMin=1.1}}
 {id=QualityLevel_1, children [Mixer_2,null,null,null]}
 {id=Random_3, children [WavePlayer_11, WavePlayer_12, WavePlayer_13, WavePlayer_14], properties {weights [1,1,1,1]}}
 {id=WavePlayer_11, properties {soundWaveAssetPtr=&/Smasher_Death01}}
 {id=WavePlayer_12, properties {soundWaveAssetPtr=&/Smasher_Death02}}
 {id=WavePlayer_13, properties {soundWaveAssetPtr=&/Smasher_Death03}}
 {id=WavePlayer_14, properties {soundWaveAssetPtr=&/Smasher_Death04}}
 {id=WavePlayer_18, properties {soundWaveAssetPtr=&/smasher_demat_01}}
 {id=WavePlayer_19, properties {soundWaveAssetPtr=&/death_sweeten_04}}
]
properties {
 attenuationSettings=/Game/Sounds/Attenuation/Fort_Monster_Smasher_Att
 bOverrideAttenuation=true
 bOverrideConcurrency=true
 maxDistance=6000
 volumeMultiplier=1.2
 attenuationOverrides {distanceAlgorithm=NaturalSound, falloffDistance=5000, attenuationShapeExtents {x=1000}}
 concurrencyOverrides {resolutionRule=PreventNew}
}
