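; Gameplay ability for the Stalker's dash attack, driving a straight dash on the avatar's player pawn.
; On activation it runs PreLaunchSetup, then checks the pawn for a configurable gameplay tag.
; If the tag matches, it applies the Stalker dash guided-motion modifier and stores the returned value.
; The ability ends immediately on either branch without waiting for the motion to finish, with debug prints for dash and end.
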(event Ability|EventActivateAbility
  (CallFunction|PerformStraightDash))

(event EventPerformStraightDash
  (CallFunction|PreLaunchSetup)
  (if (GameplayTags|HasMatchingGameplayTag (CallFunction|GetAvatarasFortPlayerPawn) (Variables|Default|GetTag%20to%20Check))
    (bind _returnvalue (Game|ApplyGuidedMotionMME (CallFunction|GetAvatarasFortPlayerPawn) "/TidalCranePrototype/Gameplay/StalkerOnTheLoose/FMM_GuidedMotion_Stalker_Dash.FMM_GuidedMotion_Stalker_Dash_C"))
    (Variables|Default|SetFMM%20EXT%20Dash _returnvalue)
    (Development|PrintString "DASHED!")
    (Ability|EndAbility)
    (else
      (Ability|EndAbility))))

(event Ability|EventOnEndAbility
  (params (bWasCancelled :type Boolean))
  (Development|PrintString "DASH ABILITY ENDED"))
