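; GA_EnergyRegen is a Gameplay Ability for the Stalker On The Loose mode that, by its name, covers energy regeneration.
; The graph only manages lifetime: activation calls the parent, waits for the Granted.StalkerOTL.Ability.End gameplay event and starts a fixed timer.
; Whichever fires first, the event or the timer, ends the ability.
; No regeneration logic appears in this graph, and the end-ability override is an unused stub.
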(event Custom|PrepareEndAbility
  (Ability|EndAbility))

(event Ability|EventActivateAbility
  (|CallParentFunction)
  (Ability|Tasks|WaitGameplayEvent "(TagName=\"Granted.StalkerOTL.Ability.End\")"
    (:then
      (Utilities|Time|SetTimerbyEvent (delegate Custom|PrepareEndAbility) 12.0))
    (:EventReceived
      (Ability|EndAbility))))

(event Ability|EventOnEndAbility
  (params (bWasCancelled :type Boolean))
  (|CallParentFunction bWasCancelled) ; This node is disabled and will not be called.\nDrag off pins to build functionality.
)
