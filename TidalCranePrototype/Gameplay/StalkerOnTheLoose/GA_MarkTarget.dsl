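; Gameplay ability for the Stalker On The Loose feature that shows the stalk-target UI and sound to a marked player for a timed hunt.
; On activation with a valid Fort player pawn, it async-creates the widget, spawns 2D audio and starts a 20-second timer.
; A parallel wait for the Marked gameplay event stores the instigator as HunterActor, without gating the UI.
; The timer fades out and removes the widget, destroys the audio and ends the ability.
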(event Ability|EventActivateAbility
  (CallFunction|GetAvatarasValidFortPlayerPawn
    (:ValidFortPlayerPawn
      (bind (_asynctaskproxy _payload) (Ability|Tasks|WaitGameplayEvent "(TagName=\"Athena.Gameplay.Status.Marked\")")
        (:then
          (bind _userwidget (Class|FortAsyncActionCreateWidgetAsync|CreateWidgetAsync "/TidalCranePrototype/Gameplay/StalkerOnTheLoose/UI/W_StalkTarget_UI.W_StalkTarget_UI_C")
            (:OnComplete
              (bind _output_get (Variables|Default|SetUser%20Widget _userwidget))
              (UserInterface|Viewport|AddtoViewport _output_get 0)
              (bind _returnvalue (Audio|SpawnSound2D (Variables|Default|GetSound) 1.3 1.0 0.0 0 false false))
              (Variables|Default|SetSpawnedAudio _returnvalue)
              (Utilities|Time|SetTimerbyEvent (delegate Custom|EndHunt) 2e+01))))
        (:EventReceived
          (Development|PrintString "RECIEVED PAYLOSD")
          (Variables|Default|SetHunterActor (.Instigator (Utilities|Struct|BreakGameplayEventData _payload))))))))

(event Custom|EndHunt
  (bind _asw_stalk_target_ui (Utilities|Casting|CastToW_StalkTarget_UI (Variables|Default|GetUser%20Widget)))
  (UserInterface|Animation|PlayAnimation _asw_stalk_target_ui (Variables|/TidalCranePrototype/Gameplay/StalkerOnTheLoose/UI/W_StalkTarget_UI.W_StalkTarget_UI_C|GetFadeOut _asw_stalk_target_ui) 0.0 1 "Forward" 1.0 false)
  (Utilities|FlowControl|Delay 2.0)
  (bind _user_widget (Variables|Default|GetUser%20Widget)
    (:then
      (Widget|RemovefromParent _user_widget))
    (:else
      _))
  (Components|DestroyComponent (Variables|Default|GetSpawnedAudio))
  (Ability|EndAbility))
