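; Gameplay ability for the Stalker On The Loose mode or character, built on a parent Mantis technique ability that likely drives melee attacks.
; Activation, end and technique start/end events defer to the parent; the sequence-name switch on technique start has no cases.
; On a technique hit it stores the target data, calls the parent, then applies the gameplay effect container matching the application tag.
; The melee hit tag switch has empty branches.
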(event Ability|EventActivateAbility
  (|CallParentFunction))

(event Ability|EventOnEndAbility
  (params (bWasCancelled :type Boolean))
  (|CallParentFunction bWasCancelled))

(event Mantis|EventOnMantisAblilityTechniqueStarted
  (params (OwningTags :type Gameplay%20Tag%20Container%20Structure) (SequenceName :type Name :default "None") (StartsSequence :type Boolean) (EndsSequence :type Boolean))
  (|CallParentFunction OwningTags SequenceName StartsSequence EndsSequence)
  (switch name SequenceName))

(event Mantis|EventOnMantisAblilityTechniqueEnded
  (params (SequenceName :type Name :default "None"))
  (|CallParentFunction SequenceName))

(event Mantis|EventOnMantisTechniqueHit
  (params (TargetDataHandle :type Gameplay%20Ability%20Target%20Data%20Handle%20Structure%20%28by%20ref%29) (ApplicationTag :type Gameplay%20Tag%20Structure))
  (bind _output_get (Variables|Default|SetTarget%20Data%20Handle TargetDataHandle))
  (|CallParentFunction _output_get ApplicationTag)
  (switch Utilities|FlowControl|Switch|SwitchonGameplayTag ApplicationTag
    (:Gameplay.Melee.Hit1
      _)
    (:Gameplay.Melee.Hit2
      _)
    (:Gameplay.Melee.Hit4.Focused
      _))
  (Ability|ApplyGameplayEffectContainer _output_get ApplicationTag))
