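; Stalker secondary gameplay ability for Stalker On The Loose that turns on a timed stalking mode.
; Activation starts target selection and immediately commits, plays an owner cue at the pawn and applies the StalkingModeOn effect, while targeting results fill TrackedActors.
; A stored timer calls PrepEndAbility, which commits again and ends the ability.
; On end it sends the Granted.StalkerOTL.Ability.End gameplay event to the avatar.
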(event Custom|PrepEndAbility
  ; Comment
  (Ability|CommitAbility)
  (Ability|EndAbility))

(event Ability|EventActivateAbility
  ; Comment
  (bind (_asynctaskproxy _targetdata) (Ability|Tasks|DoTargetSelection "None" "(TagName=\"Targeting.ApplicationTag.3\")" true)
    (:then
      (Ability|CommitAbility)
      (bind _returnvalue (Utilities|Time|SetTimerbyEvent (delegate Custom|PrepEndAbility) 15.0))
      (Variables|Default|SetMarkDurationTH _returnvalue)
      (bind _outfortplayerpawn (CallFunction|GetAvatarasValidFortPlayerPawn)
        (:ValidFortPlayerPawn
          (Ability|ExecuteGameplayCueWithParamsOnOwner (Variables|Default|GetGameplay%20Cue%20Tag_0) (Ability|GameplayCue|MakeGameplayCueParameters :NormalizedMagnitude 0.0 :RawMagnitude 0.0 :Location (Transformation|GetActorLocation _outfortplayerpawn) :Normal "0,0,0" :GameplayEffectLevel 1 :AbilityLevel 1 :bReplicateLocationWhenUsingMinimalRepProxy false))
          (Ability|ApplyGameplayEffectToOwner "/TidalCranePrototype/Gameplay/StalkerOnTheLoose/GE_StalkingModeOn.GE_StalkingModeOn_C"))))
    (:Targeted
      (Variables|Default|SetTrackedActors (Ability|TargetData|GetAllActorsfromTargetData _targetdata)))))

(event Ability|EventOnEndAbility
  (params (bWasCancelled :type Boolean))
  ; Comment
  (Ability|SendGameplayEventtoActor (CallFunction|GetAvatarasFortPlayerPawn) "(TagName=\"Granted.StalkerOTL.Ability.End\")"))
